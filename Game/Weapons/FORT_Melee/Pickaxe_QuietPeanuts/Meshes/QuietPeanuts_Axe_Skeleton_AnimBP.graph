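# Animation Blueprint for the QuietPeanuts pickaxe skeleton, adding physics-driven secondary motion to the melee weapon mesh.
# The input pose passes through DefaultSlot into a RigidBody node using the axe's physics asset, then goes to output.
# Simulation runs in base-bone space relative to the handle bone, with scaled component velocity and acceleration feeding in movement.
# The event graph holds only unconnected stubs, and the gravity override variables are unused.
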AnimBlueprint /Game/Weapons/FORT_Melee/Pickaxe_QuietPeanuts/Meshes/QuietPeanuts_Axe_Skeleton_AnimBP
target_skeleton=&/QuietPeanuts_Axe_Skeleton
graphs [
 AnimGraph {nodes [
   n1 = Input Pose()
   n2 = RigidBody(ComponentPose=Local To Component(LocalPose=Slot 'DefaultSlot'(Source=@n1))) {
    bClampLinearTranslationLimitToRefPose=true
    overridePhysicsAsset=&/QuietPeanuts_Axe_Physics
    simulationSpace=BaseBoneSpace
    alphaBoolBlend {blendOption=Linear}
    baseBoneRef {boneName=handle}
    componentLinearAccScale=<1.6 1.2 0.5>
    componentLinearVelScale=<1.6 1.2 0.5>
   }
   Output Pose(Result=Component To Local(ComponentPose=@n2)) {layerGroup=None}
  ]}
 EventGraph {nodes [
   TryGetPawnOwner()
   Event BlueprintUpdateAnimation()
  ]}
]
variables [
 Gravity Override Params {type=GravityOverrideParamsStruct}
 GravityOverride {type=Vector}
]
